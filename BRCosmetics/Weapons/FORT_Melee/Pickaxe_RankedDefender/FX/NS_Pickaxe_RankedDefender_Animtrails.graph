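# Niagara system that draws layered ribbon animation trails for the Ranked Defender pickaxe.
# Four emitters (chroma, soft dark, soft dark top, thin tip) spawn and update short-lived anim-trail ribbons from skeletal mesh bones.
# Chroma, top and tip layers are offset along RibbonFacing, and all layers take velocity from User.DebugSpeed.
# Color curves set each layer's look, System.FadeOut scales alpha, and all but the top layer get time-of-day tinting.
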NiagaraSystem /BRCosmetics/Weapons/FORT_Melee/Pickaxe_RankedDefender/FX/NS_Pickaxe_RankedDefender_Animtrails
dependencies {
 usedDataInterfaces [&NiagaraDataInterfaceColorCurve, &NiagaraDataInterfaceSkeletalMesh, &NiagaraDataInterfaceCurve]
 usedDynamicInputs [&FloatFromCurve]
 usedModules [
  &self:SystemSpawnScript.SystemScriptSource.SystemScriptGraph.NiagaraNodeAssignment_0.SetVariables_#1
  &SystemState
  &self:SystemSpawnScript.SystemScriptSource.SystemScriptGraph.NiagaraNodeAssignment_1.SetVariables_#2
 ]
}
emitters {
 NE_ChromaTrail {
  emitterUpdateScript [
   &EmitterState()
   &FN_SpawnAnimTrail()
  ]
  particleSpawnScript [
   &FN_InitializeAnimTrailBoneIndecies()
   &FN_InitializeAnimTrail(AnimTrailWidth=0.4, Lifetime=0.15, SkeletalMesh=$System.Skeletal Mesh)
   /SetVariables(Particles.Position=&AddVectorToPosition(Position=$Particles.Position, Vector=&Multiply_VectorByFloat(Vector=$Particles.RibbonFacing, Float=43))) {isSetParameters=true}
   &AddVelocity(Velocity=$User.DebugSpeed)
  ]
  particleUpdateScript [
   &FN_UpdateAnimTrail(SkeletalMesh=$System.Skeletal Mesh) {module=FN_UpdateAnimTrail001}
   &Color(
    Color=&ColorFromCurve(ColorCurve=DataInterface() {propertyValues {
       AlphaCurve {keys [<at 0 1 RCIM_Cubic>,<at 0.4649 0 RCIM_Cubic>]}
       BlueCurve {keys [<at 0 0.5 RCIM_Cubic>]}
       GreenCurve {keys [<at 0 0.5 RCIM_Cubic>]}
       RedCurve {keys [<at 0 0.5 RCIM_Cubic>]}
      }})
    Scale Alpha=$System.FadeOut
   )
   &ScaleColor(
    Color Value To Scale=$Particles.Color
    Scale Mode=::NewEnumerator1
    Scale RGBA=&MakeVector4_FromLinearColor(LinearColor=&FNTimeOfDayColor(
      Night Color=<rgba 0.3333 0.5239 1 1>
      Morning Color=<rgba 1 0.6244 0.9769 1>
      Day Color=<rgba 1 1 1 1>
      Evening Color=<rgba 1 0.6977 0.7077 1>
     ))
    ScaleRGBA=true
   )
   &SolveForcesAndVelocity(Clamp Velocity=false, Limit Acceleration=false)
   &DynamicMaterialParameters(Module.Param0WriteEnabled=true, Module.Param1WriteEnabled=true, Module.Param2WriteEnabled=true, Module.Param3WriteEnabled=true)
  ]
  rendererClasses [&NiagaraRibbonRendererProperties]
  renderers [{rendererClass=&NiagaraRibbonRendererProperties}]
 }
 NE_SoftDarkTrail {
  emitterUpdateScript [
   &EmitterState()
   &FN_SpawnAnimTrail()
  ]
  particleSpawnScript [
   &FN_InitializeAnimTrailBoneIndecies()
   &FN_InitializeAnimTrail(AnimTrailWidth=0.8, Lifetime=0.1, SkeletalMesh=$System.Skeletal Mesh)
   &AddVelocity(Velocity=$User.DebugSpeed)
  ]
  particleUpdateScript [
   &FN_UpdateAnimTrail(SkeletalMesh=$System.Skeletal Mesh) {module=FN_UpdateAnimTrail001}
   &Color(
    Color=&ColorFromCurve(ColorCurve=DataInterface() {propertyValues {
       AlphaCurve {keys [<at 0 1 RCIM_Cubic>,<at 0.4607 0.25>,<at 1 0 RCIM_Cubic>]}
       BlueCurve {keys [<at 0 0.1146>]}
       GreenCurve {keys [<at 0 0.0961>]}
       RedCurve {keys [<at 0 0.0969>]}
      }})
    Scale Alpha=$System.FadeOut
   )
   &SolveForcesAndVelocity(Clamp Velocity=false, Limit Acceleration=false)
   &ScaleColor(
    Color Value To Scale=$Particles.Color
    Scale Mode=::NewEnumerator1
    Scale RGBA=&MakeVector4_FromLinearColor(LinearColor=&FNTimeOfDayColor(Night Color=<rgba 0.6667 1.0477 2 1>, Morning Color=<rgba 1 1 1 1>, Day Color=<rgba 1 1 1 1>, Evening Color=<rgba 1 0.8391 0.4689 1>))
    ScaleRGBA=true
   )
  ]
  rendererClasses [&NiagaraRibbonRendererProperties]
  renderers [{rendererClass=&NiagaraRibbonRendererProperties}]
 }
 NE_SoftDarkTrailTop {
  emitterUpdateScript [
   &EmitterState()
   &FN_SpawnAnimTrail()
  ]
  particleSpawnScript [
   &FN_InitializeAnimTrailBoneIndecies()
   &FN_InitializeAnimTrail(AnimTrailWidth=0.3, Lifetime=0.1, SkeletalMesh=$System.Skeletal Mesh)
   /SetVariables(Particles.Position=&AddVectorToPosition(Position=$Particles.Position, Vector=&Multiply_VectorByFloat(Vector=$Particles.RibbonFacing, Float=43))) {isSetParameters=true}
   &AddVelocity(Velocity=$User.DebugSpeed)
  ]
  particleUpdateScript [
   &FN_UpdateAnimTrail(SkeletalMesh=$System.Skeletal Mesh) {module=FN_UpdateAnimTrail001}
   &Color(
    Color=&ColorFromCurve(ColorCurve=DataInterface() {propertyValues {
       AlphaCurve {keys [<at 0 0.5 RCIM_Cubic>,<at 1 0 RCIM_Cubic>]}
       BlueCurve {keys [<at 0 0.25>]}
       GreenCurve {keys [<at 0 0.1693>]}
       RedCurve {keys [<at 0 0.1589>]}
      }})
    Scale Alpha=$System.FadeOut
   )
   &SolveForcesAndVelocity(Clamp Velocity=false, Limit Acceleration=false)
  ]
  rendererClasses [&NiagaraRibbonRendererProperties]
  renderers [{rendererClass=&NiagaraRibbonRendererProperties}]
 }
 NE_TipTrail {
  emitterUpdateScript [
   &EmitterState()
   &FN_SpawnAnimTrail()
  ]
  particleSpawnScript [
   &FN_InitializeAnimTrailBoneIndecies()
   &FN_InitializeAnimTrail(AnimTrailWidth=0.05, Lifetime=0.1, SkeletalMesh=$System.Skeletal Mesh)
   /SetVariables(Particles.Position=&AddVectorToPosition(Position=$Particles.Position, Vector=&Multiply_VectorByFloat(Vector=$Particles.RibbonFacing, Float=55))) {isSetParameters=true}
   &AddVelocity(Velocity=$User.DebugSpeed)
  ]
  particleUpdateScript [
   &FN_UpdateAnimTrail(SkeletalMesh=$System.Skeletal Mesh) {module=FN_UpdateAnimTrail001}
   &Color(
    Color=&ColorFromCurve(ColorCurve=DataInterface() {propertyValues {
       AlphaCurve {keys [<at 0 1>,<at 0.3291 0 RCIM_Cubic>]}
       BlueCurve {keys [<at 0.0011 3>,<at 0.3569 0.1146>]}
       GreenCurve {keys [<at 0.0011 2.2812>,<at 0.3569 0.0352>]}
       RedCurve {keys [<at 0.0011 2.4725>,<at 0.3569 0.0658>]}
      }})
    Scale Alpha=$System.FadeOut
   )
   &ScaleColor(
    Color Value To Scale=$Particles.Color
    Scale Mode=::NewEnumerator1
    Scale RGBA=&MakeVector4_FromLinearColor(LinearColor=&FNTimeOfDayColor(Night Color=<rgba 0.3333 0.5239 1 1>, Morning Color=<rgba 1 1 1 1>, Day Color=<rgba 1 1 1 1>, Evening Color=<rgba 1 1 1 1>))
    ScaleRGBA=true
   )
   &SolveForcesAndVelocity(Clamp Velocity=false, Limit Acceleration=false)
  ]
  rendererClasses [&NiagaraRibbonRendererProperties]
  renderers [{rendererClass=&NiagaraRibbonRendererProperties}]
 }
}
summary {
 systemName=NS_Pickaxe_RankedDefender_Animtrails
 emitters [
  {emitterName=NE_ChromaTrail, rendererClasses [&NiagaraRibbonRendererProperties]}
  {emitterName=NE_SoftDarkTrail, rendererClasses [&NiagaraRibbonRendererProperties]}
  {emitterName=NE_TipTrail, rendererClasses [&NiagaraRibbonRendererProperties]}
  {emitterName=NE_SoftDarkTrailTop, rendererClasses [&NiagaraRibbonRendererProperties]}
 ]
 userVariables [
  User.DebugSpeed {type=Vector3f, description="", defaultValue {struct=Vector3f, value=<0 0 0>}}
  User.RibbonIDOffset {type=NiagaraInt32}
 ]
}
