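# Material function for the RankedDefender pickaxe that fakes a textured box interior with adjustable depth, tiled separately per axis.
# It intersects camera rays, optionally from a vertex-interpolated local camera vector, with X, Y and Z planes offset around the actor position.
# Hit points become scrolling UVs for three texture samples, with contrast-adjusted blend alphas.
# The output's last blend targets a FlattenNormal result, leaving a parallel sample lerp unconnected.
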MaterialFunction /BRCosmetics/Weapons/FORT_Melee/Pickaxe_RankedDefender/Materials/MF_TriplanarVirtualBox_TileXYZ
graph {nodes [
  n1 = LinearInterpolate(A=NamedRerouteUsage(), B=NamedRerouteUsage(), Alpha=NamedRerouteUsage())
  n2 = MaterialFunctionCall(Normal=@n1, Flatness=Constant() {r=1}) {materialFunction=&FlattenNormal}
  n3 = NamedRerouteUsage()
  FunctionOutput(None=NamedRerouteDeclaration(None=LinearInterpolate(A=@n1, B=@n2.Result, Alpha=@n3)) {name=result, nodeColor=<rgba 1 0 0.7734 1>})
  LinearInterpolate(A=@n1, B=NamedRerouteUsage(), Alpha=@n3)
  n4 = LocalPosition() {localOrigin=InstancePreSkinning}
  n5 = Normalize(VectorInput=Subtract(A=CameraPositionWS(), B=TransformPosition(None=@n4.XYZ) {transformType=TRANSFORMPOSSOURCE_World}))
  n6 = StaticBool()
  NamedRerouteDeclaration(None=StaticSwitch(
    True=VertexInterpolator(VS=@n5).PS
    False=CameraVectorWS()
    Value=FunctionInput(Preview=@n6) {bUsePreviewValueAsDefault=true, inputName=UseLocalCameraVector, inputType=FunctionInput_StaticBool, sortPriority=100}
   )) {name=CameraVector, nodeColor=<rgba 0 1 0.1562 1>}
  n7 = FunctionInput() {bUsePreviewValueAsDefault=true, inputName=Tile_UVY, inputType=FunctionInput_Scalar, sortPriority=3, previewValue=<0.25 0 0 1>}
  NamedRerouteDeclaration(None=Multiply(A=@n7) {constB=0.01}) {name=Tile_UVY, nodeColor=<rgba 1 0.5625 0 1>}
  n8 = FunctionInput() {bUsePreviewValueAsDefault=true, inputName=Depth, inputType=FunctionInput_Scalar, sortPriority=1, previewValue=<200 0 0 1>}
  NamedRerouteDeclaration(None=@n8) {name=BoxDepth, nodeColor=<rgba 0 1 0.8594 1>}
  NamedRerouteDeclaration(None=CameraPositionWS()) {name=CameraPos, nodeColor=<rgba 0.5234 1 0 1>}
  n9 = FunctionInput() {bUsePreviewValueAsDefault=true, inputName=Tile_UVX, inputType=FunctionInput_Scalar, sortPriority=2, previewValue=<0.25 0 0 1>}
  NamedRerouteDeclaration(None=Multiply(A=@n9) {constB=0.01}) {name=Tile_UVX, nodeColor=<rgba 1 0.5625 0 1>}
  n10 = Constant3Vector()
  NamedRerouteDeclaration(None=Add(
    A=FunctionInput() {bUsePreviewValueAsDefault=true, inputName=ActorOffset, sortPriority=10}
    B=TransformPosition(None=@n10) {transformType=TRANSFORMPOSSOURCE_World}
   )) {name=ActorPos, nodeColor=<rgba 0 1 0.9531 1>}
  n11 = NamedRerouteUsage()
  n12 = NamedRerouteUsage()
  n13 = Sign(None=ComponentMask(None=NamedRerouteUsage()) {r=true})
  n14 = MaterialFunctionCall(X=Multiply(A=Multiply(A=@n13) {constB=-1}, B=NamedRerouteUsage())) {materialFunction=&MakeFloat3}
  n15 = Constant3Vector() {constant=<rgba 1 0 0 0>}
  n16 = DotProduct(A=Subtract(A=Add(A=@n12, B=@n14.Result), B=NamedRerouteUsage()), B=@n15)
  n17 = DotProduct(A=@n15, B=NamedRerouteUsage())
  n18 = Add(A=NamedRerouteUsage(), B=Multiply(A=NamedRerouteUsage(), B=Divide(A=@n16, B=@n17)))
  NamedRerouteDeclaration(None=ComponentMask(None=Multiply(A=@n11, B=@n18)) {b=true, g=true}) {name=UVX, nodeColor=<rgba 0 0.6016 1 1>}
  n19 = NamedRerouteUsage()
  n20 = NamedRerouteUsage()
  n21 = Sign(None=ComponentMask(None=NamedRerouteUsage()) {g=true})
  n22 = MaterialFunctionCall(Y=Multiply(A=Multiply(A=@n21) {constB=-1}, B=NamedRerouteUsage())) {materialFunction=&MakeFloat3}
  n23 = Constant3Vector() {constant=<rgba 0 1 0 0>}
  n24 = DotProduct(A=Subtract(A=Add(A=@n20, B=@n22.Result), B=NamedRerouteUsage()), B=@n23)
  n25 = DotProduct(A=@n23, B=NamedRerouteUsage())
  n26 = Add(A=NamedRerouteUsage(), B=Multiply(A=NamedRerouteUsage(), B=Divide(A=@n24, B=@n25)))
  NamedRerouteDeclaration(None=ComponentMask(None=Multiply(A=@n19, B=@n26)) {b=true, r=true}) {name=UVY, nodeColor=<rgba 0 0.6016 1 1>}
  n27 = NamedRerouteUsage()
  n28 = NamedRerouteUsage()
  n29 = Sign(None=ComponentMask(None=NamedRerouteUsage()) {b=true})
  n30 = MaterialFunctionCall(Z=Multiply(A=Multiply(A=@n29) {constB=-1}, B=NamedRerouteUsage())) {materialFunction=&MakeFloat3}
  n31 = Constant3Vector() {constant=<rgba 0 0 1 0>}
  n32 = DotProduct(A=Subtract(A=Add(A=@n28, B=@n30.Result), B=NamedRerouteUsage()), B=@n31)
  n33 = DotProduct(A=@n31, B=NamedRerouteUsage())
  n34 = Add(A=NamedRerouteUsage(), B=Multiply(A=NamedRerouteUsage(), B=Divide(A=@n32, B=@n33)))
  NamedRerouteDeclaration(None=ComponentMask(None=Multiply(A=@n27, B=@n34)) {g=true, r=true}) {name=UVZ, nodeColor=<rgba 0 0.6016 1 1>}
  n35 = FunctionInput() {bUsePreviewValueAsDefault=true, inputName=BlendContrast, inputType=FunctionInput_Scalar, sortPriority=5, previewValue=<0.5 0 0 1>}
  NamedRerouteDeclaration(None=@n35) {name=BlendContrast, nodeColor=<rgba 1 0 0.7266 1>}
  NamedRerouteDeclaration(None=FunctionInput(Preview=TextureObject() {texture=/UVEditor/Textures/UVEditorColorGrid_Color}) {bUsePreviewValueAsDefault=true, inputName=Texture, inputType=FunctionInput_Texture2D}) {name=Galaxy_01_Texture, nodeColor=<rgba 1 0.7031 0 1>}
  n36 = NamedRerouteUsage()
  n37 = NamedRerouteUsage()
  NamedRerouteDeclaration(None=TextureSample(UVs=Add(A=@n36, B=@n37), Tex=NamedRerouteUsage()) {samplerSource=SSM_Wrap_WorldGroupSettings}.RGB) {name=SampleX, nodeColor=<rgba 1 0 0.6328 1>}
  n38 = NamedRerouteUsage()
  n39 = NamedRerouteUsage()
  NamedRerouteDeclaration(None=TextureSample(UVs=Add(A=@n38, B=@n39), Tex=NamedRerouteUsage()) {samplerSource=SSM_Wrap_WorldGroupSettings}.RGB) {name=SampleY, nodeColor=<rgba 1 0 0.6328 1>}
  n40 = NamedRerouteUsage()
  n41 = Multiply(A=NamedRerouteUsage(), B=Constant2Vector() {g=-1, r=1})
  NamedRerouteDeclaration(None=TextureSample(UVs=Add(A=@n40, B=@n41), Tex=NamedRerouteUsage()) {samplerSource=SSM_Wrap_WorldGroupSettings}.RGB) {name=SampleZ, nodeColor=<rgba 1 0 0.6328 1>}
  n42 = Abs(None=NamedRerouteUsage())
  n43 = NamedRerouteUsage()
  NamedRerouteDeclaration(None=MaterialFunctionCall(In=ComponentMask(None=@n42) {g=true}, Contrast=@n43) {materialFunction=&CheapContrast}.Result) {name=BlendAlphaY, nodeColor=<rgba 0 1 0.7422 1>}
  NamedRerouteDeclaration(None=MaterialFunctionCall(In=ComponentMask(None=@n42) {b=true}, Contrast=@n43) {materialFunction=&CheapContrast}.Result) {name=BlendAlphaZ, nodeColor=<rgba 0 0.0391 1 1>}
  n44 = FunctionInput() {bUsePreviewValueAsDefault=true, inputName=SpeedXY, inputType=FunctionInput_Vector2, sortPriority=6}
  NamedRerouteDeclaration(None=Multiply(A=ComponentMask(None=@n44) {g=true, r=true}, B=Time())) {name=Speed, nodeColor=<rgba 0 1 0.1562 1>}
  n45 = FunctionInput() {bUsePreviewValueAsDefault=true, inputName=Tile_UVZ, inputType=FunctionInput_Scalar, sortPriority=4, previewValue=<0.25 0 0 1>}
  NamedRerouteDeclaration(None=Multiply(A=@n45) {constB=0.01}) {name=Tile_UVZ, nodeColor=<rgba 1 0.5625 0 1>}
  NamedRerouteDeclaration(None=Subtract(A=NamedRerouteUsage(), B=NamedRerouteUsage())) {name=LocalCamPos, nodeColor=<rgba 1 0.7734 0 1>}
 ]}
