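# Physics asset for the RankedDefender pickaxe, a two-handed melee cosmetic weapon mesh.
# It holds a single capsule body on the Melee_TwoHanded_Handle bone, offset and rotated as a rough collision proxy for the axe.
# It has no constraints, so the weapon simulates as one rigid piece.
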PhysicsAsset /BRCosmetics/Weapons/FORT_Melee/Pickaxe_RankedDefender/Meshes/RankedDefender_Axe_PhysicsAsset
bodies {Melee_TwoHanded_Handle [{shapeType=Capsule, length=80.9484, radius=10.9516, center=<1.92 -1.3983 34.9415>, rotation=<rot -0.1138 -90 0.0109>}]}
constraints []
